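# Sound cue named for the inside of the Western rift bubble in Athena's Rift Beacon event.
# It has no root or nodes, so it holds no audio to play.
# It keeps ambient sound class, safe-zone shield attenuation and boosted volume settings.
# A concurrency override allows one instance per owner and blocks new ones while it plays.
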SoundCue /FNGameplayCues/Athena/Sounds/Events/RiftBeacon/RiftBubble/Western/Western_Inside_Cue
root=null
nodes []
properties {
 attenuationSettings=/Game/Sounds/Attenuation/Fort_SafeZone_Shield_Athena_Att
 bOverrideConcurrency=true
 maxDistance=4000
 soundClassObject=&Fort_Ambient_Athena
 volumeMultiplier=1.663
 concurrencyOverrides {bLimitToOwner=true, maxCount=1, resolutionRule=PreventNew}
}
