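# Sound cue for third-person footsteps inside the Western rift bubble in Athena, adding spur jingles from Zapper Husk step sounds.
# A random node picks one of six equally weighted spur samples, then a modulator randomizes its volume.
# It plays quietly through enemy foley sound class and walk-footstep attenuation.
# An owner-limited concurrency override caps simultaneous steps, stopping the oldest.
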SoundCue /FNGameplayCues/Athena/Sounds/Events/RiftBeacon/RiftBubble/Western/Western_Inside_Step_3P_Cue
root=Modulator_0
nodes [
 {id=Modulator_0, children [Random_0], properties {volumeMin=0.3}}
 {
  id=Random_0
  children [WavePlayer_7, WavePlayer_8, WavePlayer_9, WavePlayer_10, WavePlayer_11, WavePlayer_12]
  properties {weights [1,1,1,1,1,1]}
 }
 {id=WavePlayer_10, properties {soundWaveAssetPtr=/FNGameplayCues/Sounds/Fort_Step_Sounds/Husk/Zapper_Spurs_03}}
 {id=WavePlayer_11, properties {soundWaveAssetPtr=/FNGameplayCues/Sounds/Fort_Step_Sounds/Husk/Zapper_Spurs_05}}
 {id=WavePlayer_12, properties {soundWaveAssetPtr=/FNGameplayCues/Sounds/Fort_Step_Sounds/Husk/Zapper_Spurs_04}}
 {id=WavePlayer_7, properties {soundWaveAssetPtr=/FNGameplayCues/Sounds/Fort_Step_Sounds/Husk/Zapper_Spurs_06}}
 {id=WavePlayer_8, properties {soundWaveAssetPtr=/FNGameplayCues/Sounds/Fort_Step_Sounds/Husk/Zapper_Spurs_01}}
 {id=WavePlayer_9, properties {soundWaveAssetPtr=/FNGameplayCues/Sounds/Fort_Step_Sounds/Husk/Zapper_Spurs_02}}
]
properties {
 attenuationSettings=/Game/Sounds/Attenuation/Fort_Player_Footsteps_3P_Walk_Athena_Att
 bOverrideConcurrency=true
 maxDistance=3584
 soundClassObject=&Fort_Player_Foley_3P_Enemy_Athena
 volumeMultiplier=0.25
 concurrencyOverrides {bLimitToOwner=true, maxCount=3, resolutionRule=StopOldest}
}
