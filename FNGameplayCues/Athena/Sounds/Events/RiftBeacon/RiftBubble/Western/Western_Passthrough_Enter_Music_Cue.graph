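# Sound cue for the music sting that plays on entering the Western-themed Rift Beacon bubble in Fortnite Athena.
# A Random node picks one of two equally weighted enter-music waves.
# It plays in the ambient sound class with the safe-zone shield attenuation.
# Overridden concurrency limits it to one instance per owner, stopping the oldest.
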SoundCue /FNGameplayCues/Athena/Sounds/Events/RiftBeacon/RiftBubble/Western/Western_Passthrough_Enter_Music_Cue
root=Random_1
nodes [
 {id=Random_1, children [WavePlayer_0,WavePlayer_3], properties {weights [1,1]}}
 {id=WavePlayer_0, properties {soundWaveAssetPtr=&/Western_EnterMusic_02}}
 {id=WavePlayer_3, properties {soundWaveAssetPtr=&/Western_EnterMusic_01}}
]
properties {
 attenuationSettings=/Game/Sounds/Attenuation/Fort_SafeZone_Shield_Athena_Att
 bOverrideConcurrency=true
 maxDistance=4000
 soundClassObject=&Fort_Ambient_Athena
 volumeMultiplier=1.039
 concurrencyOverrides {bLimitToOwner=true, maxCount=1, resolutionRule=StopOldest}
}
